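; Gameplay ability for the Ashton Milo backpack that triggers its regeneration cooldown.
; When activated from a gameplay event, it calls the parent implementation with the event data, which carries any further behaviour.
; It then applies the regen cooldown gameplay effect to the owner and ends immediately.
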(event Ability|EventActivateAbilityFromEvent
  (params (EventData :type Gameplay%20Event%20Data%20Structure%20%28by%20ref%29))
  (|CallParentFunction EventData)
  (Ability|ApplyGameplayEffectToOwner "/Figment_Martin/Athena/Items/Gameplay/BackPacks/Ashton/Milo/GE_AshtonPack_Milo_Regen_Cooldown.GE_AshtonPack_Milo_Regen_Cooldown_C")
  (Ability|EndAbility))
