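; Passive ability for the Ashton Milo backpack that grants a stacking health and shield effect based on the gem stones the player holds.
; On activation it checks stacks, then waits for stone tags (StoneTag_P twice) to recheck and play a gameplay cue colored by gem number.
; GiveStacks counts matching tags on the pawn; after a debounce, ApplyEffects adds only missing stacks, then a heal effect.
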(event Ability|EventActivateAbility
  (bind _stonetag_p (Variables|Default|GetStoneTag_P))
  ;[1 1 1] When we spawn and this first fires check for stacks to initialize any health and shield we should have, After that wait for any of the gem tags to be added when they are recheck for the effects to stack.
  (CallFunction|CheckStacks)
  (Utilities|FlowControl|Sequence
    (:then_0
      (Ability|Tasks|WaitGameplayTagAdd _stonetag_p
        (:Added
          (Variables|Default|SetGemNumber 0.0) ; Number sets the color for GameplayCue
        )))
    (:then_1
      (Ability|Tasks|WaitGameplayTagAdd _stonetag_p
        (:Added
          (Variables|Default|SetGemNumber 1.0))))
    (:then_2
      (Ability|Tasks|WaitGameplayTagAdd (Variables|Default|GetStoneTag_R)
        (:Added
          (Variables|Default|SetGemNumber 2.0))))
    (:then_3
      (Ability|Tasks|WaitGameplayTagAdd (Variables|Default|GetStoneTag_O)
        (:Added
          (Variables|Default|SetGemNumber 3.0))))
    (:then_4
      (Ability|Tasks|WaitGameplayTagAdd (Variables|Default|GetStoneTag_G)
        (:Added
          (Variables|Default|SetGemNumber 4.0))))
    (:then_5
      (Ability|Tasks|WaitGameplayTagAdd (Variables|Default|GetStoneTag_Y)
        (:Added
          (Variables|Default|SetGemNumber 5.0)))))
  (CallFunction|CheckStacks)
  ;[0 1 0.1179] Apply Stone GC
  (Ability|ExecuteGameplayCueWithParamsOnOwner (Variables|Default|GetApplyStoneTag) (Ability|GameplayCue|MakeGameplayCueParameters 0.0 (Variables|Default|GetGemNumber))))

(event Custom|CheckStacks
  ;[1 1 1] Check How many of the gems we have then apply the effect using that stack.
  (Utilities|FlowControl|RetriggerableDelay 0.02) ; Make sure this isnt getting called multiple times on spawn
  (bind _stacks (CallFunction|GiveStacks))
  (if (> _stacks 0)
    (CallFunction|ApplyEffects _stacks)))

(event Custom|ApplyEffects
  (params (Stacks :type Integer))
  ;[1 1 1] Check how many stacks of this GE we should give you based on how many you already have
  (bind _returnvalue (Ability|GameplayEffect|GetActiveGameplayEffectStackCount (Variables|Default|GetGEH_HealthShields)))
  (if (> (- Stacks _returnvalue) 0)
    (bind _returnvalue_2 (Ability|ApplyGameplayEffectToOwner (Variables|Default|GetGE_HealthShields) 1 (- Stacks _returnvalue)))
    (Variables|Default|SetGEH_HealthShields _returnvalue_2)
    (Utilities|FlowControl|Delay 0.01)
    (Ability|ApplyGameplayEffectToOwner (Variables|Default|GetGE_HealUp))))

(fn GiveStacks
  (bind _returnvalue (Ability|GetAvatarActorfromActorInfo))
  (Utilities|IsValid _returnvalue
    (:Is%20Valid
      (bind _asfort_player_pawn (Utilities|Casting|CastToFortPlayerPawn _returnvalue)
        (:then
          (LocalVariables|SetPlayer _asfort_player_pawn)
          (for _array_element (LocalVariables|GetObjects)
            (if (GameplayTags|TagContainer|CompareTagAssetInterfacetoOtherTags _array_element (LocalVariables|GetPlayer))
              (Math|Integer|IncrementInt (LocalVariables|GetStacksToGive))))
          (return (LocalVariables|GetStacksToGive)))
        (:CastFailed
          _)))
    (:Is%20Not%20Valid
      _))
  (return 0))
